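# Example mono MetaSound sampler voice patch implementing the Polyphony Voice and WavePlayer control interfaces.
# Start triggers both a Wave Player, configured by the asset, pitch, start time and loop inputs, and an AD envelope.
# The player feeds the mono output through a Multiply with no envelope connection; the envelope only fires OnFinished on completion.
# Stop and both priority pins are left unwired.
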MetaSoundPatch /MetasoundPolyphonyInternal/Examples/SamplerVoice/MSP_SamplerVoice
inputs [
 Polyphony.Voice.Control.WavePlayer.Loop {type=Bool}
 Polyphony.Voice.Control.WavePlayer.LoopDuration {type=Time}
 Polyphony.Voice.Control.WavePlayer.LoopStart {type=Time}
 Polyphony.Voice.Control.WavePlayer.PitchShift {type=Float}
 Polyphony.Voice.Control.WavePlayer.StartTime {type=Time}
 Polyphony.Voice.Control.WavePlayer.WaveAsset {type=WaveAsset}
 Polyphony.Voice.In Priority {type=Int32}
 Polyphony.Voice.Start {type=Trigger}
 Polyphony.Voice.Stop {type=Trigger}
]
interfaces [Polyphony.Voice.Control.WavePlayer:1.0, Polyphony.Voice:1.0, UE.OutputFormat.Mono:1.0]
outputs [
 Polyphony.Voice.OnFinished {type=Trigger}
 Polyphony.Voice.Out Priority {type=Int32}
 "UE.OutputFormat.Mono.Audio:0" {type=Audio}
]
pages [Default {
  input_defaults {
   Polyphony.Voice.Control.WavePlayer.Loop=null
   Polyphony.Voice.Control.WavePlayer.LoopDuration=null
   Polyphony.Voice.Control.WavePlayer.LoopStart=0
   Polyphony.Voice.Control.WavePlayer.PitchShift=null
   Polyphony.Voice.Control.WavePlayer.StartTime=null
   Polyphony.Voice.Control.WavePlayer.WaveAsset=null
   Polyphony.Voice.In Priority=null
   Polyphony.Voice.Start=null
   Polyphony.Voice.Stop=null
  }
  nodes [
   ADEnvelopeADEnvelope_0 = AD Envelope.AD Envelope:Audio(Trigger=^Polyphony.Voice.Start)
   WavePlayer_0 = UE.Wave Player:Mono(
    Loop=^Polyphony.Voice.Control.WavePlayer.Loop
    Loop Duration=^Polyphony.Voice.Control.WavePlayer.LoopDuration
    Loop Start=^Polyphony.Voice.Control.WavePlayer.LoopStart
    Pitch Shift=^Polyphony.Voice.Control.WavePlayer.PitchShift
    Start Time=^Polyphony.Voice.Control.WavePlayer.StartTime
    Wave Asset=^Polyphony.Voice.Control.WavePlayer.WaveAsset
    Play=^Polyphony.Voice.Start
   ) {also {version [1,1]}}
   Multiply_0 = UE.Multiply:Audio(PrimaryOperand=@WavePlayer_0.Out Mono)
  ]
  links [{dst="@Polyphony.Voice.OnFinished", src=ADEnvelopeADEnvelope_0.On Done}, {dst="@UE.OutputFormat.Mono.Audio:0", src=Multiply_0.Out}]
 }]
