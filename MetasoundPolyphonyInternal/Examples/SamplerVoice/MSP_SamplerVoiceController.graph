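# MetaSound patch from the polyphony sampler examples that turns a sample index and trigger into a voice control stream.
# The index is clamped to the Samples array, and a trigger compare against -1 gates the incoming trigger.
# A passed trigger fetches the wave asset, which becomes a WavePlayer voice parameter keyed by that index.
# The same trigger drives the voice controller's Trigger and, through a zero-delay pipe, its Stop.
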MetaSoundPatch /MetasoundPolyphonyInternal/Examples/SamplerVoice/MSP_SamplerVoiceController
inputs [
 Index {type=Int32}
 Samples {type=WaveAsset:Array}
 Trigger {type=Trigger}
]
outputs [Output {type=Polyphony::VoiceControlStream}]
pages [Default {
  input_defaults {Index=0, Trigger=null, Samples []}
  nodes [
   n1 = Clamp.Clamp:Int32(In=^Index, Max=Array.GetLastIndex:WaveAsset:Array(Array=^Samples).Last Index) {also {defaults {Min=-1}}}
   n2 = TriggerCompare.Clamp:Int32(Compare=^Trigger, A=@n1.Value) {also {defaults {B=-1, Type=3}}}
   MetasoundPolyphonyInternalTriggerParametersVoiceController_0 = MetasoundPolyphonyInternal.TriggerParametersVoiceController(
    Key Index=@n1.Value
    Voice Parameters=MetasoundPolyphonyInternal.ConversionWaveAssetToPolyphony::VoiceParameters(In=Array.Get:WaveAsset:Array(Array=^Samples, Index=@n1.Value, Trigger=@n2.True).Element) {also {defaults {Parameter Name=Polyphony.Voice.Control.WavePlayer.WaveAsset}, version [0,1]}}.Voice Parameters
    Stop=UE.Pipe(In=@n2.True) {also {defaults {Delay Time=0}}}.Out
    Trigger=@n2.True
   ) {also {version [0,1]}}
  ]
  links [{dst="@Output", src=MetasoundPolyphonyInternalTriggerParametersVoiceController_0.Voice Control Stream}]
 }]
