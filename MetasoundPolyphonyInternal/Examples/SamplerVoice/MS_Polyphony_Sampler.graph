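# Looping mono MetaSound source that demonstrates polyphonic sample playback with the MetasoundPolyphonyInternal nodes.
# On play, three TriggerRepeat timers with different periods each drive an MSP_SamplerVoiceController, which has its own index into a shared wave asset array.
# The three voice control streams are merged into one and fed to a PolyphonyPatchPlayer, which renders MSP_SamplerVoice patch voices to the mono output.
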MetaSoundSource /MetasoundPolyphonyInternal/Examples/SamplerVoice/MS_Polyphony_Sampler
inputs [UE.Source.OnPlay {type=Trigger}]
interfaces [UE.OutputFormat.Mono:1.0, UE.Source:1.1]
outputs ["UE.OutputFormat.Mono.Audio:0" {type=Audio}]
pages [Default {
  input_defaults {UE.Source.OnPlay=null}
  nodes [
   n1 = VariableAccessor.WaveAsset:Array(Variable=InitVariable.WaveAsset:Array().Variable)
   MetasoundPolyphonyInternalMergeVoiceControlStreams3_0 = "MetasoundPolyphonyInternal.MergeVoiceControlStreams(3)"(
    In 1=&/MSP_SamplerVoiceController(Trigger=UE.TriggerRepeat(Start=^UE.Source.OnPlay).RepeatOut, Samples=@n1.Value) {also {defaults {Index=1}}}.Output
    In 0=&MSP_SamplerVoiceController(Trigger=UE.TriggerRepeat(Start=^UE.Source.OnPlay) {also {defaults {Period=0.8}}}.RepeatOut, Samples=@n1.Value).Output
    In 2=&MSP_SamplerVoiceController(Trigger=UE.TriggerRepeat(Start=^UE.Source.OnPlay) {also {defaults {Period=0.4}}}.RepeatOut, Samples=@n1.Value) {also {defaults {Index=2}}}.Output
   ) {also {version [0,1]}}
   MetasoundPolyphonyInternalPolyphonyPatchPlayer_0 = MetasoundPolyphonyInternal.PolyphonyPatchPlayer:Mono(Voice Control Stream=@MetasoundPolyphonyInternalMergeVoiceControlStreams3_0.Out) {also {defaults {MetaSound Asset=&/MSP_SamplerVoice}, version [0,1]}}
  ]
  links [{dst="@UE.OutputFormat.Mono.Audio:0", src=MetasoundPolyphonyInternalPolyphonyPatchPlayer_0.Out Mono}]
 }]
properties {
 assetImportData=&self:AssetImportData
 bLooping=true
 duration=10000
 sampleRate=48000
}
